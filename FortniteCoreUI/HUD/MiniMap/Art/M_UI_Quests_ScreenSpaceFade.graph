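# Translucent UI material for quest elements in the minimap HUD, building a screen-space fade mask with optional rounded corners.
# Screen-aligned vertical distance from center, offset and scaled by falloff parameters, gives a saturated mask rising toward top and bottom edges.
# A sharpened rounded-square gradient on 9-slice widget UVs multiplies that mask, and a scalar parameter blends between plain and rounded results.
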Material /FortniteCoreUI/HUD/MiniMap/Art/M_UI_Quests_ScreenSpaceFade
graph {
 properties {
  bUsedWithStaticMesh=false
  blendMode=BLEND_Translucent
  materialDomain=MD_UI
  pixelDepthOffsetMode=PDOM_Legacy
  usedShadingModels=Default Lit
 }
 nodes [
  Constant() {r=1}
  n1 = MaterialFunctionCall() {materialFunction=&ScreenAlignedUVs}
  n2 = ComponentMask(None=Abs(None=Add(A=@n1."X 100%, Y 100%") {constB=-0.5})) {g=true}
  n3 = ScalarParameter() {defaultValue=0.45, group=Item, parameterName=falloff_position}
  n4 = Saturate(None=Multiply(A=Subtract(A=@n2, B=@n3), B=ScalarParameter() {defaultValue=10, group=Item, parameterName=falloff_Strenght}) {constB=32})
  n5 = MaterialFunctionCall() {materialFunction=&GetUserInterfaceUV}
  n6 = ScalarParameter() {defaultValue=0.25, group=Rounded corners, parameterName=Curvature}
  n7 = Multiply(
   A=ComponentMask(None=MaterialFunctionCall!(UV=@n5."9-Slice UV (No Tiling)", Size, Curvature=@n6) {materialFunction=/Game/Effects/Fort_Effects/MaterialFunctions/MF_RoundedSquare}.Linear Gradient) {g=true}
   B=ScalarParameter() {defaultValue=32, group=Rounded corners, parameterName=EdgeSharpness}
  )
  LinearInterpolate(A=@n4, B=Multiply(A=@n4, B=Saturate(None=@n7)), Alpha=ScalarParameter() {parameterName=UseRoundedCorner})
 ]
}
