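# Translucent UI material for minimap arrow buttons, drawing either a procedural arrow or, via the IsArrow static switch, a radial gradient.
# The arrow is built from two mirrored, rotated UV bands joined by a hard center-split mask, with optional stroke outline.
# Hover animation blends shape, sharpness and gradient density between A/B parameters.
# Chained lerps select base, hover, pressed and disabled fills, whose alpha scales the shape mask.
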Material /FortniteCoreUI/HUD/MiniMap/Art/M_Ui_ArrowButtons
graph {
 properties {
  bUsedWithStaticMesh=false
  blendMode=BLEND_Translucent
  materialDomain=MD_UI
  pixelDepthOffsetMode=PDOM_Legacy
  usedShadingModels=Default Lit
 }
 nodes [
  n1 = VectorParameter() {group=Colors, parameterName=Base_Fill, defaultValue=<rgba 1 1 1 0.5>}
  n2 = VectorParameter() {group=Colors, parameterName=Hover_Fill, defaultValue=<rgba 1 1 1 1>}
  n3 = ScalarParameter() {group=Animators, parameterName=Hover_Animate, sliderMax=1}
  n4 = VectorParameter() {group=Colors, parameterName=Pressed_Fill, defaultValue=<rgba 1 1 1 1>}
  n5 = ScalarParameter() {group=Animators, parameterName=Pressed_Animate, sliderMax=1}
  n6 = VectorParameter() {group=Colors, parameterName=Disabled_Fill, defaultValue=<rgba 0.0467 0.107 0.3185 1>}
  n7 = ScalarParameter() {group=Animators, parameterName=bIsDisabled, sliderMax=1}
  LinearInterpolate(A=LinearInterpolate(A=LinearInterpolate(A=@n1.RGB, B=@n2.RGB, Alpha=@n3), B=@n4.RGB, Alpha=@n5), B=@n6.RGB, Alpha=@n7)
  n8 = Multiply(A=Add(A=TextureCoordinate()) {constB=-0.5}) {constB=1024}
  NamedRerouteDeclaration(None=ComponentMask(None=Saturate(None=@n8)) {g=true}) {name=Mask, nodeColor=<rgba 0.2891 0 1 1>}
  n9 = TextureCoordinate()
  n10 = ScalarParameter() {defaultValue=0.0625, group=Arrow, parameterName=Rotation}
  n11 = Rotator(Coordinate=@n9, Time=@n10) {speed=6.2832}
  n12 = LinearInterpolate(
   A=ScalarParameter() {defaultValue=-0.625, group=Arrow, parameterName=XShift_A}
   B=ScalarParameter() {defaultValue=-0.625, group=Arrow, parameterName=XShift_B}
   Alpha=ScalarParameter() {group=Animators, parameterName=Hover_Animate, sliderMax=1}
  )
  n13 = Add(A=LinearInterpolate(
    A=ScalarParameter() {defaultValue=0.45, group=Arrow, parameterName=Thickness_Y_A}
    B=ScalarParameter() {defaultValue=0.375, group=Arrow, parameterName=Thickness_Y_B}
    Alpha=ScalarParameter() {group=Animators, parameterName=Hover_Animate, sliderMax=1}
   )) {constB=-0.5}
  n14 = Add(A=Abs(None=Add(A=@n11, B=@n12) {constB=-0.5}), B=@n13)
  n15 = ComponentMask(None=@n14) {r=true}
  n16 = Rotator(Coordinate=@n9, Time=Multiply(A=@n10) {constB=-1}) {speed=6.2832}
  n17 = Add(A=Abs(None=Add(A=@n16, B=@n12) {constB=-0.5}), B=@n13)
  n18 = ComponentMask(None=@n17) {r=true}
  n19 = NamedRerouteUsage()
  n20 = ScalarParameter() {defaultValue=64, group=Arrow, parameterName=Sharpness_A}
  n21 = ScalarParameter() {defaultValue=64, group=Arrow, parameterName=Sharpness_B}
  n22 = ScalarParameter() {group=Animators, parameterName=Hover_Animate, sliderMax=1}
  n23 = Saturate(None=Multiply(A=LinearInterpolate(A=@n15, B=@n18, Alpha=@n19), B=LinearInterpolate(A=@n20, B=@n21, Alpha=@n22)))
  n24 = TextureCoordinate()
  n25 = ScalarParameter() {defaultValue=0.0625, group=Arrow, parameterName=Rotation}
  n26 = Add(A=@n25) {constB=0.5}
  n27 = Add(A=LinearInterpolate(
    A=ScalarParameter() {defaultValue=0.125, group=Arrow, parameterName=Length_Y_A}
    B=ScalarParameter() {defaultValue=0.125, group=Arrow, parameterName=Length_Y_B}
    Alpha=ScalarParameter() {group=Animators, parameterName=Hover_Animate, sliderMax=1}
   )) {constB=-0.5}
  n28 = Add(A=Add(A=Rotator(Coordinate=@n24, Time=@n26) {speed=6.2832}) {constB=-0.5}, B=@n27)
  n29 = ComponentMask(None=@n28) {g=true}
  n30 = Multiply(A=@n25) {constB=-1}
  n31 = Add(A=Add(A=Rotator(Coordinate=@n24, Time=@n30) {speed=6.2832}) {constB=-0.5}, B=@n27)
  n32 = ComponentMask(None=@n31) {g=true}
  n33 = NamedRerouteUsage()
  n34 = LinearInterpolate(
   A=ScalarParameter() {defaultValue=64, group=Arrow, parameterName=Sharpness_A}
   B=ScalarParameter() {defaultValue=64, group=Arrow, parameterName=Sharpness_B}
   Alpha=ScalarParameter() {group=Animators, parameterName=Hover_Animate, sliderMax=1}
  )
  n35 = Saturate(None=Multiply(A=LinearInterpolate(A=@n29, B=@n32, Alpha=@n33), B=@n34))
  n36 = Saturate(None=DotProduct(A=OneMinus(None=@n23), B=OneMinus(None=@n35)))
  n37 = OneMinus(None=@n36)
  n38 = ScalarParameter() {defaultValue=-0.0312, group=Arrow, parameterName=StrokeThickness_A}
  n39 = ScalarParameter() {defaultValue=-0.0312, group=Arrow, parameterName=StrokeThickness_B}
  n40 = ScalarParameter() {group=Animators, parameterName=Hover_Animate, sliderMax=1}
  n41 = LinearInterpolate(A=@n38, B=@n39, Alpha=@n40)
  n42 = ComponentMask(None=Add(A=@n14, B=@n41)) {r=true}
  n43 = ComponentMask(None=Add(A=@n17, B=@n41)) {r=true}
  n44 = NamedRerouteUsage()
  n45 = Saturate(None=Multiply(A=LinearInterpolate(A=@n42, B=@n43, Alpha=@n44), B=@n34))
  n46 = LinearInterpolate(A=@n38, B=@n39, Alpha=@n40)
  n47 = ComponentMask(None=Add(A=@n28, B=@n46)) {g=true}
  n48 = ComponentMask(None=Add(A=@n31, B=@n46)) {g=true}
  n49 = NamedRerouteUsage()
  n50 = Saturate(None=Multiply(A=LinearInterpolate(A=@n47, B=@n48, Alpha=@n49), B=@n34))
  n51 = Saturate(None=DotProduct(A=OneMinus(None=@n45), B=OneMinus(None=@n50)))
  NamedRerouteDeclaration(None=StaticSwitchParameter(True=LinearInterpolate(B=@n37, Alpha=@n51), False=@n36) {parameterName=IsStroke}) {name=ArrowMask, nodeColor=<rgba 0.6406 1 0 1>}
  n52 = LinearInterpolate(A=LinearInterpolate(A=LinearInterpolate(A=@n1.A, B=@n2.A, Alpha=@n3), B=@n4.A, Alpha=@n5), B=@n6.A, Alpha=@n7)
  n53 = TextureCoordinate()
  n54 = AppendVector(A=ScalarParameter() {defaultValue=0.5, parameterName=Rad_XPosition}, B=ScalarParameter() {defaultValue=0.5, parameterName=Rad_YPosition})
  n55 = ScalarParameter() {defaultValue=0.5, parameterName=Rad_Size}
  n56 = LinearInterpolate(
   A=ScalarParameter() {defaultValue=1, parameterName=Rad_Density_A}
   B=ScalarParameter() {defaultValue=1, parameterName=Rad_Density_B}
   Alpha=ScalarParameter() {group=Animators, parameterName=Hover_Animate, sliderMax=1}
  )
  StaticSwitchParameter(
   True=LinearInterpolate(B=@n52, Alpha=NamedRerouteUsage())
   False=LinearInterpolate(
    B=@n52
    Alpha=MaterialFunctionCall(UVs=@n53, CenterPosition=@n54, Radius=@n55, Density=@n56) {materialFunction=&RadialGradientExponential}.RadialGradientExponential
   )
  ) {defaultValue=true, parameterName=IsArrow}
 ]
}
