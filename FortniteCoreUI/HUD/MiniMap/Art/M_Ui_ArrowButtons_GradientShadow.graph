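# Translucent UI material for a gradient shadow, apparently behind minimap HUD arrow buttons, going by its name and folder.
# Hover_Animate blends color and alpha between the RGBA_Base and RGBA_Hover parameters.
# The blended alpha is multiplied by a horizontal UV gradient and a rounded-square mask.
# That mask feeds 9-slice UVs and a Curvature parameter into MF_RoundedSquare, then sharpens the result with EdgeSharpness and saturates it.
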Material /FortniteCoreUI/HUD/MiniMap/Art/M_Ui_ArrowButtons_GradientShadow
graph {
 properties {
  bUsedWithStaticMesh=false
  blendMode=BLEND_Translucent
  materialDomain=MD_UI
  pixelDepthOffsetMode=PDOM_Legacy
  usedShadingModels=Default Lit
 }
 nodes [
  n1 = VectorParameter() {parameterName=RGBA_Base, defaultValue=<rgba 0 0.0056 0.1329 0.3>}
  n2 = VectorParameter() {parameterName=RGBA_Hover, defaultValue=<rgba 0.5501 0.549 0.6094 0.4>}
  n3 = ScalarParameter() {parameterName=Hover_Animate}
  LinearInterpolate(A=@n1.RGB, B=@n2.RGB, Alpha=@n3)
  n4 = TextureCoordinate()
  n5 = ComponentMask(None=MaterialFunctionCall!(
    UV=MaterialFunctionCall() {materialFunction=&GetUserInterfaceUV}."9-Slice UV (No Tiling)"
    Size
    Curvature=ScalarParameter() {group=Rounded corners, parameterName=Curvature}
   ) {materialFunction=/Game/Effects/Fort_Effects/MaterialFunctions/MF_RoundedSquare}.Linear Gradient) {g=true}
  n6 = ScalarParameter() {defaultValue=32, group=Rounded corners, parameterName=EdgeSharpness}
  Multiply(A=Multiply(A=LinearInterpolate(A=@n1.A, B=@n2.A, Alpha=@n3), B=ComponentMask(None=@n4) {r=true}), B=Saturate(None=Multiply(A=@n5, B=@n6)))
 ]
}
